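; Backpack part modifier Blueprint for the PastelGlazeGift back bling that configures its idle VFX after initialization.
; OnPostResIn sets front-end and front-end preview bool parameters on the idle VFX component based on whether it is in the front end.
; It then reads the owning pawn's backpack Material variant index, activating the Niagara idle effect for Mat0 and deactivating it for Mat1.
; The other event handlers are empty or disabled stubs.
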(event EventOnPostResIn
  (bind _returnvalue (CallFunction|GetIdleVFXComponent))
  (bind _returnvalue_2 (CallFunction|GetIdleVFXComponent))
  (|CallParentFunction)
  (if (CustomCosmeticModifier|IsinFrontEnd)
    (Effects|Components|ParticleSystem|SetBoolParameter _returnvalue "bIsFrontEnd" true)
    (Effects|Components|ParticleSystem|SetBoolParameter _returnvalue "bIsFrontEndPreview" true)
    (else
      (Effects|Components|ParticleSystem|SetBoolParameter _returnvalue "bIsFrontEnd" false)
      (Effects|Components|ParticleSystem|SetBoolParameter _returnvalue "bIsFrontEndPreview" false)))
  ;[0.3024 0.1295 0.6719 0.5] FX - Reactivity check
  (bind _asfort_player_pawn (Utilities|Casting|CastToFortPlayerPawn (Actor|GetOwner)))
  (bind _index (CosmeticLoadout|GetVariantChannelActiveIndexforSlotActor "(TagName=\"LoadoutSlot.Backpack\")" _asfort_player_pawn "(TagName=\"Cosmetics.Variant.Channel.Material\")" (Utilities|Array|MakeArray "(TagName=\"Cosmetics.Variant.Property.Mat0\")" "(TagName=\"Cosmetics.Variant.Property.Mat1\")") false false))
  (if (Utilities|Select :Index _index :Option%200 true :Option%201 false)
    (bind _asniagara_particle_system_component (Utilities|Casting|CastToNiagaraComponent _returnvalue_2))
    (Components|Activation|Activate _asniagara_particle_system_component false)
    (else
      (bind _asniagara_particle_system_component_2 (Utilities|Casting|CastToNiagaraComponent _returnvalue_2))
      (Components|Activation|Deactivate _asniagara_particle_system_component_2))))

(event EventOnDayPhaseChanged
  (params (CurrentDayPhase :type EFortDayPhase%20Enum :default "Morning") (PreviousDayPhase :type EFortDayPhase%20Enum :default "Morning") (bAtCreation :type Boolean)))

(event EventOnNewPlayerWeaponType
  (params (NewType :type EFortItemType%20Enum :default "WorldItem") (OldType :type EFortItemType%20Enum :default "WorldItem"))
  (|CallParentFunction NewType OldType) ; This node is disabled and will not be called.\nDrag off pins to build functionality.
)

(event EventOnWeaponChanged
  (params (NewWeapon :type Fort%20Weapon%20Object%20Reference) (OldWeapon :type Fort%20Weapon%20Object%20Reference)))

(event EventOnWeaponFired)

(event EventOnNewWeapons
  (params (WeaponsAsItems :type Array%20of%20Fort%20Item%20Object%20References)))

(event EventOnCurieStateAttached
  (params (Tag :type Gameplay%20Tag%20Structure%20%28by%20ref%29)))

(event EventOnCurieStateDetached
  (params (Tag :type Gameplay%20Tag%20Structure%20%28by%20ref%29)))

(event EventOnBeginSkydiving
  (|CallParentFunction) ; This node is disabled and will not be called.\nDrag off pins to build functionality.
)

(event EventOnEndSkydiving
  (|CallParentFunction) ; This node is disabled and will not be called.\nDrag off pins to build functionality.
)

(event EventOnEnterVehicle
  (params (VehicleEntered :type Fort%20Vehicle%20Interface%20Interface%20%28by%20ref%29))
  (|CallParentFunction VehicleEntered) ; This node is disabled and will not be called.\nDrag off pins to build functionality.
)

(event EventOnExitVehicle
  (|CallParentFunction) ; This node is disabled and will not be called.\nDrag off pins to build functionality.
)

(event EventOnTargetingOrFirstPersonCameraChanged)

(event EventOnHitPawn
  (params (SelfActor :type Actor%20Object%20Reference) (OtherActor :type Actor%20Object%20Reference) (NormalImpulse :type Vector :default "0, 0, 0") (Hit :type Hit%20Result%20Structure%20%28by%20ref%29)))

(fn ConstructionScript
  (|CallParentFunction))
