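# Physics asset for the PastelGlazeGift backpack cosmetic, giving its left and right wings capsule bodies for jointed physical motion.
# Each wing links three bone chains (c, d, e) to its 01 body through swing-limited constraints.
# The c and e joints lock twist, while the d joints set a narrow twist limit.
# The e chain runs 01, 2, 4, 3, out of numeric order, and no constraint joins the two wings.
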PhysicsAsset /BRCosmetics/Accessories/FORT_Backpacks/Backpack_F_MED_PastelGlazeGift/Meshes/F_MED_PastelGlazeGift_Physics
bodies {
 dyn_wing_l_01 [{
   shapeType=Capsule
   length=8.7079
   radius=6.0416
   shapeName=dyn_wing_l_01_capsule
   center=<0.4601 -0.0144 -3.0917>
   rotation=<rot 0 114.0718 -57.438>
  }]
 dyn_wing_l_c_2 [{shapeType=Capsule, length=5.5792, radius=4.3678, shapeName=dyn_wing_l_c_2_capsule, rotation=<rot 60 0 0>}]
 dyn_wing_l_c_3 [{shapeType=Capsule, length=7.3488, radius=5.8284, shapeName=dyn_wing_l_c_3_capsule, rotation=<rot 90 0 0>}]
 dyn_wing_l_d_2 [{shapeType=Capsule, length=14.9182, radius=5.5746, shapeName=dyn_wing_l_d_2_capsule, rotation=<rot 90 0 0>}]
 dyn_wing_l_d_3 [{shapeType=Capsule, length=12.2044, radius=7.6259, shapeName=dyn_wing_l_d_3_capsule, rotation=<rot 90 0 0>}]
 dyn_wing_l_e_2 [{shapeType=Capsule, length=5.3444, radius=3.7823, shapeName=dyn_wing_l_e_2_capsule, rotation=<rot 90 0 0>}]
 dyn_wing_l_e_3 [{shapeType=Capsule, length=9.9133, radius=3.225, shapeName=dyn_wing_l_e_3_capsule, rotation=<rot 90 0 0>}]
 dyn_wing_l_e_4 [{shapeType=Capsule, length=5.9076, radius=3.7451, shapeName=dyn_wing_l_e_4_capsule, rotation=<rot 90 0 0>}]
 dyn_wing_r_01 [{
   shapeType=Capsule
   length=8.7079
   radius=6.0416
   shapeName=dyn_wing_r_01_capsule
   center=<-0.4601 0.0144 3.0917>
   rotation=<rot 0 114.0718 -57.438>
  }]
 dyn_wing_r_c_2 [{shapeType=Capsule, length=5.5792, radius=4.3678, shapeName=dyn_wing_r_c_2_capsule, rotation=<rot 60 0 0>}]
 dyn_wing_r_c_3 [{shapeType=Capsule, length=7.3488, radius=5.8284, shapeName=dyn_wing_r_c_3_capsule, rotation=<rot 90 0 0>}]
 dyn_wing_r_d_2 [{shapeType=Capsule, length=14.9182, radius=5.5746, shapeName=dyn_wing_r_d_2_capsule, rotation=<rot 90 0 0>}]
 dyn_wing_r_d_3 [{shapeType=Capsule, length=12.2044, radius=7.6259, shapeName=dyn_wing_r_d_3_capsule, rotation=<rot 90 0 0>}]
 dyn_wing_r_e_2 [{shapeType=Capsule, length=5.3444, radius=3.7823, shapeName=dyn_wing_r_e_2_capsule, rotation=<rot 90 0 0>}]
 dyn_wing_r_e_3 [{shapeType=Capsule, length=9.9133, radius=3.225, shapeName=dyn_wing_r_e_3_capsule, rotation=<rot 90 0 0>}]
 dyn_wing_r_e_4 [{shapeType=Capsule, length=5.9076, radius=3.7451, shapeName=dyn_wing_r_e_4_capsule, rotation=<rot 90 0 0>}]
}
constraints [
 {bone1Name=dyn_wing_l_c_2, bone2Name=dyn_wing_l_c_3, swing1LimitDegrees=15, swing2LimitDegrees=25, twistLimitDegrees=45, twistMotion=Locked}
 {bone1Name=dyn_wing_l_d_2, bone2Name=dyn_wing_l_d_3, swing1LimitDegrees=25, swing2LimitDegrees=35, twistLimitDegrees=5}
 {bone1Name=dyn_wing_l_e_4, bone2Name=dyn_wing_l_e_2, swing1LimitDegrees=15, swing2LimitDegrees=25, twistLimitDegrees=45, twistMotion=Locked}
 {bone1Name=dyn_wing_l_e_3, bone2Name=dyn_wing_l_e_4, swing1LimitDegrees=15, swing2LimitDegrees=25, twistLimitDegrees=45, twistMotion=Locked}
 {bone1Name=dyn_wing_r_c_2, bone2Name=dyn_wing_r_c_3, swing1LimitDegrees=15, swing2LimitDegrees=25, twistLimitDegrees=45, twistMotion=Locked}
 {bone1Name=dyn_wing_r_d_2, bone2Name=dyn_wing_r_d_3, swing1LimitDegrees=25, swing2LimitDegrees=35, twistLimitDegrees=5}
 {bone1Name=dyn_wing_r_e_4, bone2Name=dyn_wing_r_e_2, swing1LimitDegrees=15, swing2LimitDegrees=25, twistLimitDegrees=45, twistMotion=Locked}
 {bone1Name=dyn_wing_r_e_3, bone2Name=dyn_wing_r_e_4, swing1LimitDegrees=15, swing2LimitDegrees=20, twistLimitDegrees=45, twistMotion=Locked}
 {bone1Name=dyn_wing_r_d_3, bone2Name=dyn_wing_r_01, swing1LimitDegrees=25, swing2LimitDegrees=35, twistLimitDegrees=5}
 {bone1Name=dyn_wing_r_e_2, bone2Name=dyn_wing_r_01, swing1LimitDegrees=15, swing2LimitDegrees=25, twistLimitDegrees=45, twistMotion=Locked}
 {bone1Name=dyn_wing_r_c_3, bone2Name=dyn_wing_r_01, swing1LimitDegrees=15, swing2LimitDegrees=25, twistLimitDegrees=45, twistMotion=Locked}
 {bone1Name=dyn_wing_l_d_3, bone2Name=dyn_wing_l_01, swing1LimitDegrees=25, swing2LimitDegrees=35, twistLimitDegrees=5}
 {bone1Name=dyn_wing_l_e_2, bone2Name=dyn_wing_l_01, swing1LimitDegrees=15, swing2LimitDegrees=25, twistLimitDegrees=45, twistMotion=Locked}
 {bone1Name=dyn_wing_l_c_3, bone2Name=dyn_wing_l_01, swing1LimitDegrees=15, swing2LimitDegrees=25, twistLimitDegrees=45, twistMotion=Locked}
]
