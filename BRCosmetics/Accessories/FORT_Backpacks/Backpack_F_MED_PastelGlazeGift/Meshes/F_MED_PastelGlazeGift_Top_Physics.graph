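# Physics asset for the top mesh of the PastelGlazeGift backpack cosmetic, covering its left and right dynamic wing bones.
# Each side has a tiny capsule on its master bone and mirrored, larger capsules on the two wing segments.
# Constraints chain master to first to second segment with tight swing and twist limits.
# The looser swing sits at the base joint on the left but the outer joint on the right.
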PhysicsAsset /BRCosmetics/Accessories/FORT_Backpacks/Backpack_F_MED_PastelGlazeGift/Meshes/F_MED_PastelGlazeGift_Top_Physics
bodies {
 dyn_wing_base_l_01 [{
   shapeType=Capsule
   length=4.4852
   radius=3.0667
   shapeName=dyn_wing_base_l_01_capsule
   center=<1.167 0.3785 2.3025>
   rotation=<rot -11.2003 -90 37.8716>
  }]
 dyn_wing_base_l_02 [{
   shapeType=Capsule
   length=7.2736
   radius=5.8143
   shapeName=dyn_wing_base_l_02_capsule
   center=<6.0536 -0.7504 1.4631>
   rotation=<rot 0 101.8071 -64.5913>
  }]
 dyn_wing_base_l_master [{shapeType=Capsule, length=0.505, radius=0.505, shapeName=dyn_wing_base_l_master_capsule, rotation=<rot 0 -90 0>}]
 dyn_wing_base_r_01 [{
   shapeType=Capsule
   length=4.4852
   radius=3.0667
   shapeName=dyn_wing_base_r_01_capsule
   center=<-1.167 -0.3785 -2.3025>
   rotation=<rot -11.2003 -90 37.8716>
  }]
 dyn_wing_base_r_02 [{
   shapeType=Capsule
   length=7.2736
   radius=5.8143
   shapeName=dyn_wing_base_r_02_capsule
   center=<-6.0536 0.7504 -1.4631>
   rotation=<rot 0 101.8071 -64.5913>
  }]
 dyn_wing_base_r_master [{shapeType=Capsule, length=0.505, radius=0.505, shapeName=dyn_wing_base_r_master_capsule, rotation=<rot 0 -90 0>}]
}
constraints [
 {bone1Name=dyn_wing_base_l_01, bone2Name=dyn_wing_base_l_master, swing1LimitDegrees=5, swing2LimitDegrees=5, twistLimitDegrees=2}
 {bone1Name=dyn_wing_base_l_02, bone2Name=dyn_wing_base_l_01, swing1LimitDegrees=2, swing2LimitDegrees=2, twistLimitDegrees=2}
 {bone1Name=dyn_wing_base_r_01, bone2Name=dyn_wing_base_r_master, swing1LimitDegrees=2, swing2LimitDegrees=2, twistLimitDegrees=2}
 {bone1Name=dyn_wing_base_r_02, bone2Name=dyn_wing_base_r_01, swing1LimitDegrees=5, swing2LimitDegrees=5, twistLimitDegrees=2}
]
