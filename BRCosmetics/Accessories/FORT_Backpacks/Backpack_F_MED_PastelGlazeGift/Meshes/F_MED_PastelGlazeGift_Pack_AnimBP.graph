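# Anim Blueprint for the PastelGlazeGift back bling, posing its wings in response to player movement and flight.
# It copies the mesh pose, then stacks additive bone rotations: movement and spine twist drive the ground branch, while clavicle, spine and delayed glider-open values drive the skydive branch.
# Each branch ends in two RigidBody simulations, and bIsSkyDiving selects between them.
# The event graph computes these inputs each update, adjusting gravity and physics for skydiving and gliding.
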AnimBlueprint /BRCosmetics/Accessories/FORT_Backpacks/Backpack_F_MED_PastelGlazeGift/Meshes/F_MED_PastelGlazeGift_Pack_AnimBP
target_skeleton=&/F_MED_PastelGlazeGift_Pack_Skeleton
graphs [
 AnimGraph {nodes [
   Input Pose()
   "Transform (Modify) Bone - Bone: dyn_wing_r_01"(Alpha=1, Translation="0,0,0", Rotation="1.330,-35.647,56.007", Scale="1,1,1") {
    rotationMode=BMM_Additive
    translationMode=BMM_Additive
    alphaBoolBlend {blendInTime=0.3, blendOption=Linear, blendOutTime=0.7}
    alphaScaleBiasClamp {bInterpResult=true, interpSpeedDecreasing=1.4, interpSpeedIncreasing=5, scale=0.7}
    boneToModify {boneName=dyn_wing_r_01}
    rotation=<rot 1.33 -35.647 56.007>
   }
   "Transform (Modify) Bone - Bone: dyn_wing_l_01"(Alpha=1, Translation="0,0,0", Rotation="-40.049,20.569,59.202", Scale="1,1,1") {
    rotationMode=BMM_Additive
    translationMode=BMM_Additive
    alphaBoolBlend {blendInTime=0.3, blendOption=Linear, blendOutTime=0.7}
    alphaScaleBiasClamp {bInterpResult=true, interpSpeedDecreasing=1.4, interpSpeedIncreasing=5, scale=0.7}
    boneToModify {boneName=dyn_wing_l_01}
    rotation=<rot -40.049 20.569 59.202>
   }
   "Transform (Modify) Bone - Bone: dyn_wing_base_r_02"(Alpha=1, Translation="0,0,0", Rotation="-9.339,32.048,32.539", Scale="1,1,1") {
    rotationMode=BMM_Additive
    translationMode=BMM_Additive
    alphaBoolBlend {blendInTime=0.3, blendOption=Linear, blendOutTime=0.7}
    alphaScaleBiasClamp {bInterpResult=true, interpSpeedDecreasing=1.5, interpSpeedIncreasing=5, scale=0.7}
    boneToModify {boneName=dyn_wing_base_r_02}
    rotation=<rot -9.339 32.048 32.539>
   }
   "Transform (Modify) Bone - Bone: dyn_wing_base_l_02"(Alpha=1, Translation="0,0,0", Rotation="6.060,-53.890,25.959", Scale="1,1,1") {
    rotationMode=BMM_Additive
    translationMode=BMM_Additive
    alphaBoolBlend {blendInTime=0.3, blendOption=Linear, blendOutTime=0.7}
    alphaScaleBiasClamp {bInterpResult=true, interpSpeedDecreasing=1.5, interpSpeedIncreasing=5, scale=0.7}
    boneToModify {boneName=dyn_wing_base_l_02}
    rotation=<rot 6.06 -53.89 25.959>
   }
   n1 = "Transform (Modify) Bone - Bone: dyn_wing_base_r_01"(Alpha=1, Translation="0,0,0", Rotation="68.091,102.263,-26.661", Scale="1,1,1") {
    rotationMode=BMM_Additive
    translationMode=BMM_Additive
    alphaBoolBlend {blendInTime=0.3, blendOption=Linear, blendOutTime=0.7}
    alphaScaleBiasClamp {bInterpResult=true, interpSpeedIncreasing=5, scale=0.5}
    boneToModify {boneName=dyn_wing_base_r_01}
    rotation=<rot 68.091 102.263 -26.661>
   }
   n2 = "Transform (Modify) Bone - Bone: dyn_wing_r_01"(ComponentPose=@n1, Alpha=1, Translation="0,0,0", Rotation="-9.551,-23.784,51.589", Scale="1,1,1") {
    rotationMode=BMM_Additive
    translationMode=BMM_Additive
    alphaBoolBlend {blendInTime=0.3, blendOption=Linear, blendOutTime=0.7}
    alphaScaleBiasClamp {bInterpResult=true, interpSpeedIncreasing=5, scale=0.5}
    boneToModify {boneName=dyn_wing_r_01}
    rotation=<rot -9.551 -23.784 51.589>
   }
   n3 = "Transform (Modify) Bone - Bone: dyn_wing_base_l_01"(ComponentPose=@n2, Alpha=1, Translation="0,0,0", Rotation="-68.091,-102.263,-26.661", Scale="1,1,1") {
    rotationMode=BMM_Additive
    translationMode=BMM_Additive
    alphaBoolBlend {blendInTime=0.3, blendOption=Linear, blendOutTime=0.7}
    alphaScaleBiasClamp {bInterpResult=true, interpSpeedIncreasing=5, scale=0.5}
    boneToModify {boneName=dyn_wing_base_l_01}
    rotation=<rot -68.091 -102.263 -26.661>
   }
   "Transform (Modify) Bone - Bone: dyn_wing_l_01"(ComponentPose=@n3, Alpha=1, Translation="0,0,0", Rotation="6.701,-24.458,65.385", Scale="1,1,1") {
    rotationMode=BMM_Additive
    translationMode=BMM_Additive
    alphaBoolBlend {blendInTime=0.3, blendOption=Linear, blendOutTime=0.7}
    alphaScaleBiasClamp {bInterpResult=true, interpSpeedIncreasing=5, scale=0.5}
    boneToModify {boneName=dyn_wing_l_01}
    rotation=<rot 6.701 -24.458 65.385>
   }
   n4 = "Transform (Modify) Bone - Bone: dyn_wing_base_l_master"(
    ComponentPose=Local To Component(LocalPose=Copy Pose From Mesh())
    Alpha=1
    Translation="0,0,0"
    Rotation=Get Return Value Rotation l() -> (Return Value Rotation l:Rotator)
    Scale="1,1,1"
   ) {
    rotationMode=BMM_Additive
    translationMode=BMM_Additive
    alphaBoolBlend {blendOption=Linear}
    alphaScaleBiasClamp {bInterpResult=true, scale=0.25}
    boneToModify {boneName=dyn_wing_base_l_master}
   }
   n5 = "Transform (Modify) Bone - Bone: dyn_wing_base_r_master"(ComponentPose=@n4, Alpha=1, Translation="0,0,0", Rotation=Get Return Value Rotation_r() -> (Return Value Rotation_r:Rotator), Scale="1,1,1") {
    rotationMode=BMM_Additive
    translationMode=BMM_Additive
    alphaBoolBlend {blendOption=Linear}
    alphaScaleBiasClamp {bInterpResult=true, scale=0.5}
    boneToModify {boneName=dyn_wing_base_r_master}
   }
   n6 = "Transform (Modify) Bone - Bone: dyn_wing_base_l_master"!(
    ComponentPose=@n5
    Alpha=1
    Translation="0,0,0"
    Rotation_Roll:float=0
    Rotation_Pitch:float=15
    Rotation_Yaw:float=Get rotValue_FlightState_L() -> (rotValue_FlightState_L:double)
    Scale="1,1,1"
   ) -> (Pose) {
    rotationMode=BMM_Additive
    translationMode=BMM_Additive
    alphaBoolBlend {blendInTime=0.3, blendOption=Linear, blendOutTime=0.7}
    alphaScaleBiasClamp {bInterpResult=true}
    boneToModify {boneName=dyn_wing_base_l_master}
    rotation {yaw=-25}
   }
   n7 = "Transform (Modify) Bone - Bone: dyn_wing_base_r_master"!(
    ComponentPose=@n6
    Alpha=1
    Translation="0,0,0"
    Rotation_Roll:float=0
    Rotation_Pitch:float=15
    Rotation_Yaw:float=Get rotValue_FlightState_R() -> (rotValue_FlightState_R:double)
    Scale="1,1,1"
   ) -> (Pose) {
    rotationMode=BMM_Additive
    translationMode=BMM_Additive
    alphaBoolBlend {blendInTime=0.3, blendOption=Linear, blendOutTime=0.7}
    alphaScaleBiasClamp {bInterpResult=true}
    boneToModify {boneName=dyn_wing_base_r_master}
    rotation {yaw=-15}
   }
   n8 = "Transform (Modify) Bone - Bone: dyn_wing_base_l_master"!(ComponentPose=@n7, bAlphaBoolEnabled:bool=True, Translation="0,0,0", Rotation:Rotator="20,-30,25", Scale="1,1,1") -> (Pose) {
    alphaInputType=Bool
    rotationMode=BMM_Additive
    translationMode=BMM_Additive
    alphaBoolBlend {blendInTime=0.3, blendOption=Linear, blendOutTime=1}
    alphaScaleBiasClamp {bInterpResult=true}
    boneToModify {boneName=dyn_wing_base_l_master}
    rotation=<rot 20 -30 25>
   }
   n9 = "Transform (Modify) Bone - Bone: dyn_wing_base_r_master"!(ComponentPose=@n8, bAlphaBoolEnabled:bool=True, Translation="0,0,0", Rotation:Rotator="-20,30,25", Scale="1,1,1") -> (Pose) {
    alphaInputType=Bool
    rotationMode=BMM_Additive
    translationMode=BMM_Additive
    alphaBoolBlend {blendInTime=0.3, blendOption=Linear, blendOutTime=1}
    alphaScaleBiasClamp {bInterpResult=true}
    boneToModify {boneName=dyn_wing_base_r_master}
    rotation=<rot -20 30 25>
   }
   n10 = "Transform (Modify) Bone - Bone: dyn_wing_base_r_master"!(ComponentPose=@n9, Alpha:float=1, Translation="0,0,0", Rotation="-10,15,25", Scale="1,1,1") -> (Pose) {
    rotationMode=BMM_Additive
    translationMode=BMM_Additive
    alphaBoolBlend {blendInTime=0.3, blendOption=Linear, blendOutTime=0.5}
    alphaScaleBiasClamp {bInterpResult=true, scale=0.5}
    boneToModify {boneName=dyn_wing_base_r_master}
    rotation=<rot -10 15 25>
   }
   n11 = "Transform (Modify) Bone - Bone: dyn_wing_base_l_master"!(ComponentPose=@n10, Alpha:float=1, Translation="0,0,0", Rotation="10,0,25", Scale="1,1,1") -> (Pose) {
    rotationMode=BMM_Additive
    translationMode=BMM_Additive
    alphaBoolBlend {blendInTime=0.3, blendOption=Linear, blendOutTime=0.5}
    alphaScaleBiasClamp {bInterpResult=true, scale=0.5}
    boneToModify {boneName=dyn_wing_base_l_master}
    rotation {pitch=10, roll=25}
   }
   n12 = Get Para Glide Open Glider_0() -> (Para Glide Open Glider_0:bool)
   n13 = "Transform (Modify) Bone - Bone: dyn_wing_base_l_master"!(ComponentPose=@n11, bAlphaBoolEnabled:bool=@n12, Translation="0,0,0", Rotation="0,-25,15", Scale="1,1,1") -> (Pose) {
    alphaInputType=Bool
    rotationMode=BMM_Additive
    translationMode=BMM_Additive
    alphaBoolBlend {blendInTime=0.2, blendOption=Linear, blendOutTime=0.3}
    alphaScaleBiasClamp {bInterpResult=true, interpSpeedIncreasing=15}
    boneToModify {boneName=dyn_wing_base_l_master}
    rotation {roll=15, yaw=-25}
   }
   n14 = "Transform (Modify) Bone - Bone: dyn_wing_l_e_2"!(ComponentPose=@n13, bAlphaBoolEnabled:bool=@n12, Translation="0,0,0", Rotation="0,35,55", Scale="1,1,1") -> (Pose) {
    alphaInputType=Bool
    rotationMode=BMM_Additive
    translationMode=BMM_Additive
    alphaBoolBlend {blendInTime=0.2, blendOption=Linear, blendOutTime=0.3}
    alphaScaleBiasClamp {bInterpResult=true, interpSpeedIncreasing=15}
    boneToModify {boneName=dyn_wing_l_e_2}
    rotation {roll=55, yaw=35}
   }
   n15 = "Get Return ValueSpine_05 Rotation Z (Yaw)"() -> ("Return ValueSpine_05 Rotation Z (Yaw)":float)
   n16 = "Transform (Modify) Bone - Bone: dyn_wing_base_l_master"!(
    ComponentPose=@n14
    bAlphaBoolEnabled=True
    Translation="0,0,0"
    Rotation_Roll:float=@n15
    Rotation_Pitch:float=0
    Rotation_Yaw:float=0
    Scale="1,1,1"
   ) -> (Pose) {
    alphaInputType=Bool
    rotationMode=BMM_Additive
    translationMode=BMM_Additive
    alphaBoolBlend {blendInTime=0.3, blendOption=Linear, blendOutTime=0.5}
    alphaScaleBiasClamp {bInterpResult=true, interpSpeedIncreasing=15}
    boneToModify {boneName=dyn_wing_base_l_master}
    rotation {roll=25}
   }
   n17 = "Transform (Modify) Bone - Bone: dyn_wing_l_e_2"!(
    ComponentPose=@n16
    bAlphaBoolEnabled=True
    Translation="0,0,0"
    Rotation_Roll:float=@n15
    Rotation_Pitch:float=0
    Rotation_Yaw:float=15
    Scale="1,1,1"
   ) -> (Pose) {
    alphaInputType=Bool
    rotationMode=BMM_Additive
    translationMode=BMM_Additive
    alphaBoolBlend {blendInTime=0.3, blendOption=Linear, blendOutTime=0.3}
    alphaScaleBiasClamp {bInterpResult=true, interpSpeedIncreasing=15}
    boneToModify {boneName=dyn_wing_l_e_2}
    rotation {roll=85, yaw=35}
   }
   n18 = "Transform (Modify) Bone - Bone: dyn_wing_base_r_master"!(ComponentPose=@n17, bAlphaBoolEnabled:bool=@n12, Translation="0,0,0", Rotation="0,15,15", Scale="1,1,1") -> (Pose) {
    alphaInputType=Bool
    rotationMode=BMM_Additive
    translationMode=BMM_Additive
    alphaBoolBlend {blendInTime=0.2, blendOption=Linear, blendOutTime=0.3}
    alphaScaleBiasClamp {bInterpResult=true, interpSpeedIncreasing=15}
    boneToModify {boneName=dyn_wing_base_r_master}
    rotation {roll=15, yaw=15}
   }
   n19 = "Transform (Modify) Bone - Bone: dyn_wing_r_e_2"!(ComponentPose=@n18, bAlphaBoolEnabled:bool=@n12, Translation="0,0,0", Rotation="0,-35,55", Scale="1,1,1") -> (Pose) {
    alphaInputType=Bool
    rotationMode=BMM_Additive
    translationMode=BMM_Additive
    alphaBoolBlend {blendInTime=0.2, blendOption=Linear, blendOutTime=0.3}
    alphaScaleBiasClamp {bInterpResult=true, interpSpeedIncreasing=15}
    boneToModify {boneName=dyn_wing_r_e_2}
    rotation {roll=55, yaw=-35}
   }
   n20 = "Transform (Modify) Bone - Bone: dyn_wing_base_r_master"!(
    ComponentPose=@n19
    bAlphaBoolEnabled=True
    Translation="0,0,0"
    Rotation_Roll:float=@n15
    Rotation_Pitch:float=0
    Rotation_Yaw:float=0
    Scale="1,1,1"
   ) -> (Pose) {
    alphaInputType=Bool
    rotationMode=BMM_Additive
    translationMode=BMM_Additive
    alphaBoolBlend {blendInTime=0.3, blendOption=Linear, blendOutTime=0.5}
    alphaScaleBiasClamp {bInterpResult=true, interpSpeedIncreasing=15}
    boneToModify {boneName=dyn_wing_base_r_master}
    rotation {roll=25}
   }
   n21 = "Transform (Modify) Bone - Bone: dyn_wing_r_e_2"!(
    ComponentPose=@n20
    bAlphaBoolEnabled=True
    Translation="0,0,0"
    Rotation_Roll:float=@n15
    Rotation_Pitch:float=0
    Rotation_Yaw:float=-15
    Scale="1,1,1"
   ) -> (Pose) {
    alphaInputType=Bool
    rotationMode=BMM_Additive
    translationMode=BMM_Additive
    alphaBoolBlend {blendInTime=0.3, blendOption=Linear, blendOutTime=0.3}
    alphaScaleBiasClamp {bInterpResult=true, interpSpeedIncreasing=15}
    boneToModify {boneName=dyn_wing_r_e_2}
    rotation {roll=85, yaw=-35}
   }
   n22 = Get Override World Gravity() -> (Override World Gravity:Vector)
   n23 = RigidBody!(
    ComponentPose=@n21
    Alpha=Get alphaSkyCond() -> (alphaSkyCond:double)
    OverrideWorldGravity:Vector=@n22
    ExternalForce="1.5,1.5,1.5"
    ComponentLinearAccScale:Vector=Get Component Linear Acc Scale() -> (Component Linear Acc Scale:Vector)
   ) -> (Pose) {
    bOverrideWorldGravity=true
    overridePhysicsAsset=&/F_MED_PastelGlazeGift_Physics
    simulationSpace=BaseBoneSpace
    alphaBoolBlend {blendOption=Linear}
    alphaScaleBiasClamp {bInterpResult=true, interpSpeedDecreasing=5}
    baseBoneRef {boneName=pack}
    componentLinearAccScale=<0.8 0.8 1>
    componentLinearVelScale=<1 1 1>
    overrideWorldGravity {z=-800}
    simSpaceSettings {maxAngularAcceleration=1100, maxAngularVelocity=10, maxLinearAcceleration=250, maxLinearVelocity=150, velocityScaleZ=0.5, worldAlpha=1}
   }
   n24 = RigidBody!(ComponentPose=@n23, Alpha=0.3, OverrideWorldGravity=@n22, ExternalForce) -> (Pose) {
    bOverrideWorldGravity=true
    overridePhysicsAsset=&/F_MED_PastelGlazeGift_Top_Physics
    simulationSpace=BaseBoneSpace
    alphaBoolBlend {blendOption=Linear}
    alphaScaleBiasClamp {bInterpResult=true}
    baseBoneRef {boneName=pack}
    componentLinearAccScale=<0.7 0.7 0.5>
    componentLinearVelScale=<0.25 0.25 0.5>
    overrideWorldGravity {z=-980}
    simSpaceSettings {maxAngularVelocity=15, maxLinearAcceleration=1750, maxLinearVelocity=10, worldAlpha=1}
   }
   Output Pose(Result=Blend Poses by bool(BlendPose_0=Component To Local(ComponentPose=@n24), BlendPose_1=Use cached pose 'dynamicCache'(), bActiveValue=Get bIsSkyDiving()) {blendType=Linear}) {layerGroup=None}
   n25 = Get leftMov_fvalue() -> (leftMov_fvalue:double)
   n26 = "Transform (Modify) Bone - Bone: dyn_wing_base_l_02"(
    ComponentPose=Local To Component(LocalPose=Copy Pose From Mesh())
    Alpha=@n25
    Translation="0,0,0"
    Rotation="0.601,-45.256,-7.041"
    Scale="1,1,1"
   ) {
    rotationMode=BMM_Additive
    translationMode=BMM_Additive
    alphaBoolBlend {blendOption=Linear}
    alphaScaleBiasClamp {bInterpResult=true, interpSpeedDecreasing=5, interpSpeedIncreasing=7}
    boneToModify {boneName=dyn_wing_base_l_02}
    rotation=<rot 0.601 -45.256 -7.041>
   }
   n27 = "Transform (Modify) Bone - Bone: dyn_wing_l_01"(ComponentPose=@n26, Alpha=@n25, Translation="0,0,0", Rotation="-0.1,-15,17", Scale="1,1,1") {
    rotationMode=BMM_Additive
    translationMode=BMM_Additive
    alphaBoolBlend {blendOption=Linear}
    alphaScaleBiasClamp {bInterpResult=true, interpSpeedDecreasing=5, interpSpeedIncreasing=7}
    boneToModify {boneName=dyn_wing_l_01}
    rotation=<rot -0.1 -15 17>
   }
   n28 = "Transform (Modify) Bone - Bone: dyn_wing_base_r_master"!(ComponentPose=@n27, Alpha:float=@n25, Translation="0,0,0", Rotation:Rotator="-5,-25,-10", Scale="1,1,1") -> (Pose) {
    rotationMode=BMM_Additive
    translationMode=BMM_Additive
    alphaBoolBlend {blendOption=Linear}
    alphaScaleBiasClamp {bInterpResult=true, interpSpeedDecreasing=5, interpSpeedIncreasing=7, scale=0.5}
    boneToModify {boneName=dyn_wing_base_r_master}
    rotation=<rot -5 -25 -10>
   }
   n29 = Get RightMov_fvalue() -> (RightMov_fvalue:double)
   n30 = "Transform (Modify) Bone - Bone: dyn_wing_base_r_02"(ComponentPose=@n28, Alpha=@n29, Translation="0,0,0", Rotation="0.601,45.256,-7.041", Scale="1,1,1") {
    rotationMode=BMM_Additive
    translationMode=BMM_Additive
    alphaBoolBlend {blendOption=Linear}
    alphaScaleBiasClamp {bInterpResult=true, interpSpeedDecreasing=5, interpSpeedIncreasing=7}
    boneToModify {boneName=dyn_wing_base_r_02}
    rotation=<rot 0.601 45.256 -7.041>
   }
   n31 = "Transform (Modify) Bone - Bone: dyn_wing_r_01"(ComponentPose=@n30, Alpha=@n29, Translation="0,0,0", Rotation="5,10,5", Scale="1,1,1") {
    rotationMode=BMM_Additive
    translationMode=BMM_Additive
    alphaBoolBlend {blendOption=Linear}
    alphaScaleBiasClamp {bInterpResult=true, interpSpeedDecreasing=5, interpSpeedIncreasing=7}
    boneToModify {boneName=dyn_wing_r_01}
    rotation=<rot 5 10 5>
   }
   n32 = "Transform (Modify) Bone - Bone: dyn_wing_base_l_master"!(ComponentPose=@n31, Alpha:float=@n29, Translation="0,0,0", Rotation:Rotator="10,25,-15", Scale="1,1,1") -> (Pose) {
    rotationMode=BMM_Additive
    translationMode=BMM_Additive
    alphaBoolBlend {blendOption=Linear}
    alphaScaleBiasClamp {bInterpResult=true, interpSpeedDecreasing=5, interpSpeedIncreasing=7, scale=0.5}
    boneToModify {boneName=dyn_wing_base_l_master}
    rotation=<rot 10 25 -15>
   }
   n33 = Get forwardMov_fvalue() -> (forwardMov_fvalue:double)
   n34 = "Transform (Modify) Bone - Bone: dyn_wing_base_r_01"(ComponentPose=@n32, Alpha=@n33, Translation="0,0,0", Rotation="21.314,33.764,-8.037", Scale="1,1,1") {
    rotationMode=BMM_Additive
    translationMode=BMM_Additive
    alphaBoolBlend {blendInTime=0.3, blendOption=Linear, blendOutTime=0.7}
    alphaScaleBiasClamp {bInterpResult=true, interpSpeedIncreasing=5}
    boneToModify {boneName=dyn_wing_base_r_01}
    rotation=<rot 21.314 33.764 -8.037>
   }
   n35 = "Transform (Modify) Bone - Bone: dyn_wing_r_01"(ComponentPose=@n34, Alpha=@n33, Translation="0,0,0", Rotation="-7.240,29.508,28.471", Scale="1,1,1") {
    rotationMode=BMM_Additive
    translationMode=BMM_Additive
    alphaBoolBlend {blendInTime=0.3, blendOption=Linear, blendOutTime=0.7}
    alphaScaleBiasClamp {bInterpResult=true, interpSpeedIncreasing=5}
    boneToModify {boneName=dyn_wing_r_01}
    rotation=<rot -7.24 29.508 28.471>
   }
   n36 = Get Melee_spineRotation X() -> (Melee_spineRotation X:float)
   n37 = "Transform (Modify) Bone - Bone: dyn_wing_base_l_master"(ComponentPose=@n35, Alpha=@n36, Translation="0,0,0", Rotation="5,-10,20", Scale="1,1,1") {
    rotationMode=BMM_Additive
    translationMode=BMM_Additive
    alphaBoolBlend {blendInTime=0.3, blendOption=Linear, blendOutTime=1}
    alphaScaleBiasClamp {bInterpResult=true}
    boneToModify {boneName=dyn_wing_base_l_master}
    rotation=<rot 5 -10 20>
   }
   n38 = "Transform (Modify) Bone - Bone: dyn_wing_base_l_01"(ComponentPose=@n37, Alpha=@n33, Translation="0,0,0", Rotation="-10.762,-59.784,-13.951", Scale="1,1,1") {
    rotationMode=BMM_Additive
    translationMode=BMM_Additive
    alphaBoolBlend {blendInTime=0.3, blendOption=Linear, blendOutTime=0.7}
    alphaScaleBiasClamp {bInterpResult=true, interpSpeedIncreasing=5}
    boneToModify {boneName=dyn_wing_base_l_01}
    rotation=<rot -10.762 -59.784 -13.951>
   }
   n39 = "Transform (Modify) Bone - Bone: dyn_wing_l_01"(ComponentPose=@n38, Alpha=@n33, Translation="0,0,0", Rotation="-7.24,-29.508,28.471", Scale="1,1,1") {
    rotationMode=BMM_Additive
    translationMode=BMM_Additive
    alphaBoolBlend {blendInTime=0.3, blendOption=Linear, blendOutTime=0.7}
    alphaScaleBiasClamp {bInterpResult=true, interpSpeedIncreasing=5}
    boneToModify {boneName=dyn_wing_l_01}
    rotation=<rot -7.24 -29.508 28.471>
   }
   n40 = "Transform (Modify) Bone - Bone: dyn_wing_base_r_master"(ComponentPose=@n39, Alpha=@n36, Translation="0,0,0", Rotation="-5,10,20", Scale="1,1,1") {
    rotationMode=BMM_Additive
    translationMode=BMM_Additive
    alphaBoolBlend {blendInTime=0.3, blendOption=Linear, blendOutTime=1}
    alphaScaleBiasClamp {bInterpResult=true}
    boneToModify {boneName=dyn_wing_base_r_master}
    rotation=<rot -5 10 20>
   }
   n41 = "Transform (Modify) Bone - Bone: dyn_wing_r_01"(ComponentPose=@n40, Alpha=1, Translation="0,0,0", Rotation="1.33,-35.647,35", Scale="1,1,1") {
    rotationMode=BMM_Additive
    translationMode=BMM_Additive
    alphaBoolBlend {blendInTime=0.3, blendOption=Linear, blendOutTime=0.7}
    alphaScaleBiasClamp {bInterpResult=true, interpSpeedDecreasing=1.5, interpSpeedIncreasing=5, scale=0.7}
    boneToModify {boneName=dyn_wing_r_01}
    rotation=<rot 1.33 -35.647 35>
   }
   n42 = "Transform (Modify) Bone - Bone: dyn_wing_base_r_01"(ComponentPose=@n41, Alpha=1, Translation="0,0,0", Rotation="3.359,-39.921,-11.054", Scale="1,1,1") {
    rotationMode=BMM_Additive
    translationMode=BMM_Additive
    alphaBoolBlend {blendInTime=0.3, blendOption=Linear, blendOutTime=0.7}
    alphaScaleBiasClamp {bInterpResult=true, interpSpeedDecreasing=1.5, interpSpeedIncreasing=5, scale=0.7}
    boneToModify {boneName=dyn_wing_base_r_01}
    rotation=<rot 3.359 -39.921 -11.054>
   }
   n43 = "Transform (Modify) Bone - Bone: dyn_wing_l_01"(ComponentPose=@n42, Alpha=1, Translation="0,0,0", Rotation="-40.049,20.569,25", Scale="1,1,1") {
    rotationMode=BMM_Additive
    translationMode=BMM_Additive
    alphaBoolBlend {blendInTime=0.3, blendOption=Linear, blendOutTime=0.7}
    alphaScaleBiasClamp {bInterpResult=true, interpSpeedDecreasing=1.5, interpSpeedIncreasing=5, scale=0.7}
    boneToModify {boneName=dyn_wing_l_01}
    rotation=<rot -40.049 20.569 25>
   }
   n44 = "Transform (Modify) Bone - Bone: dyn_wing_base_l_01"(ComponentPose=@n43, Alpha=1, Translation="0,0,0", Rotation="3.359,39.921,-11.054", Scale="1,1,1") {
    rotationMode=BMM_Additive
    translationMode=BMM_Additive
    alphaBoolBlend {blendInTime=0.3, blendOption=Linear, blendOutTime=0.7}
    alphaScaleBiasClamp {bInterpResult=true, interpSpeedDecreasing=1.5, interpSpeedIncreasing=5, scale=0.7}
    boneToModify {boneName=dyn_wing_base_l_01}
    rotation=<rot 3.359 39.921 -11.054>
   }
   Save cached pose 'preBaseAnimTrasform'(Pose=Component To Local(ComponentPose=@n44)) {cacheName=preBaseAnimTrasform}
   n45 = Get Override World Gravity()
   n46 = RigidBody(
    ComponentPose=Local To Component(LocalPose=Use cached pose 'preBaseAnimTrasform'())
    Alpha=0.65
    OverrideWorldGravity=@n45
    ExternalForce="1,1,1"
   ) {
    bOverrideWorldGravity=true
    overridePhysicsAsset=&F_MED_PastelGlazeGift_Physics
    simulationSpace=BaseBoneSpace
    alphaBoolBlend {blendOption=Linear}
    alphaScaleBiasClamp {bInterpResult=true}
    baseBoneRef {boneName=pack}
    componentAppliedLinearAccClamp=<100 100 100>
    componentLinearAccScale=<0.1 0.1 0.1>
    componentLinearVelScale=<0.1 0.1 0.1>
    overrideWorldGravity {z=-800}
    simSpaceSettings {
     maxAngularAcceleration=91.2646
     maxAngularVelocity=5
     maxLinearAcceleration=33
     maxLinearVelocity=2
     velocityScaleZ=0.1
     worldAlpha=1
     externalLinearVelocity {z=6}
    }
   }
   n47 = RigidBody(ComponentPose=@n46, Alpha=0.5, OverrideWorldGravity=@n45, ExternalForce="1.5,1.5,1.5") {
    bOverrideWorldGravity=true
    overridePhysicsAsset=&F_MED_PastelGlazeGift_Top_Physics
    simulationSpace=BaseBoneSpace
    alphaBoolBlend {blendOption=Linear}
    alphaScaleBiasClamp {bInterpResult=true}
    baseBoneRef {boneName=pack}
    componentAppliedLinearAccClamp=<1500 1500 1500>
    componentLinearAccScale=<0.7 0.7 0.5>
    componentLinearVelScale=<0.25 0.25 0.5>
    overrideWorldGravity {z=-980}
    simSpaceSettings {maxAngularAcceleration=1300, maxAngularVelocity=70, maxLinearAcceleration=3, maxLinearVelocity=11, velocityScaleZ=0.9, worldAlpha=1}
   }
   Save cached pose 'dynamicCache'(Pose=Component To Local(ComponentPose=@n47)) {cacheName=dynamicCache}
   Reroute Node(InputPin:double=Reroute Node()) -> (OutputPin:double)
   Reroute Node(InputPin=Reroute Node())
   Get wingsCorrectIsPara_tempVal() -> (wingsCorrectIsPara_tempVal:double)
  ]}
 EventGraph {nodes [
   n1 = Event BlueprintUpdateAnimation()
   > Set Delta Time X(Delta Time X=@n1.DeltaTimeX)
   > n2 = PlayerMovement(AnimInstance=Self-Reference(), XDotProductCoefficient=1, YDotProductCoefficient=1, ZDotProductCoefficient=1)
   > n3 = Sequence!(execute) -> (then_0, then_1, then_2:Exec, then_3:Exec, then_4:Exec, then_5:Exec, then_6:Exec)
   n4 = GetOwnerSkeletalMeshComponent()
   Set Melee_spineRotation X(
    execute=@n3.then_6
    Melee_spineRotation X:float=GetAlphaFromReferenceTransformDelta(
     SkeletalMeshComponent=@n4
     BoneName=spine_05
     BaseName=spine_01
     DeltaComponentType=QuaternionTwist
     InputRange=(X=-24,Y=-33)
     TwistAxis="1,0,0"
    )
   ) -> (Output_Get:float)
   > Branch(Condition=Get bIsSkydiveFloating())
   n5 = GetSocketTransform!(self=@n4, InSocketName=spine_05, TransformSpace=RTS_Component) -> (
    ReturnValue_Location_X:double
    ReturnValue_Location_Y:double
    ReturnValue_Location_Z:double
    ReturnValue_Rotation_Roll:float
    ReturnValue_Rotation_Pitch:float
    ReturnValue_Rotation_Yaw:float
    ReturnValue_Scale:Vector
   )
   > Set rotValue_FlightState_L(rotValue_FlightState_L:double="Clamp (Float)"(Value=@n5, Min=-15, Max=11)) -> (Output_Get:double)
   > Set rotValue_FlightState_R(rotValue_FlightState_R:double="Clamp (Float)"(Value=@n5, Min=-6.2, Max=30)) -> (Output_Get:double)
   > "Set Return ValueSpine_05 Rotation Z (Yaw)"("Return ValueSpine_05 Rotation Z (Yaw)":float=MapRangeClamped(Value=@n5.ReturnValue_Rotation_Yaw, InRangeA=75, InRangeB=96, OutRangeA=7, OutRangeB=15)) -> (Output_Get:float)
   n6 = Get bIsSkyDiving()
   n7 = "To Boolean (Integer)"!(InInt:int=Round(A="Max (Float)"(A=@n2, B=0)) {is=title K2Node_CallFunction}) -> (ReturnValue:bool)
   n8 = NOT Boolean(A=Get bIsJumping())
   n9 = Branch(execute=@n3.then_0, Condition=AND Boolean(A=AND Boolean(A=@n7, B=@n8), B=NOT Boolean(A=@n6)))
   n10 = Get bIsSkyDiving()
   > n11 = Branch(Condition=AND Boolean(A=Get bIsCrouching(), B=NOT Boolean(A=@n10)))
   n12 = Set RightMov_fvalue(execute=@n9.else, RightMov_fvalue:double=0) -> (Output_Get:double)
   n13 = Get Delta Time X()
   n14 = MapRangeClamped(Value=FInterpTo(Current=@n12, Target=float + float(A=@n12, B=1), DeltaTime=@n13, InterpSpeed=4), InRangeA=0, InRangeB=1)
   n15 = Set RightMov_fvalue(execute=@n11.else, RightMov_fvalue=@n14)
   Set RightMov_fvalue(
    execute=@n11.then
    RightMov_fvalue=FInterpTo(Current=@n15, Target=float * float(A=@n14, B=0.92) {is=title K2Node_PromotableOperator}, DeltaTime=Get Delta Time X(), InterpSpeed=25)
   )
   n16 = Get bIsSkyDiving()
   n17 = "Min (Float)"(A:double=@n2, B:double=0) -> (ReturnValue:double) {is=title K2Node_CommutativeAssociativeBinaryOperator}
   n18 = "To Boolean (Integer)"(InInt=Round(A="Absolute (Float)"(A=@n17)))
   n19 = NOT Boolean(A=Get bIsJumping())
   n20 = Branch(execute=@n3.then_2, Condition=AND Boolean(A=AND Boolean(A=@n18, B=@n19), B=NOT Boolean(A=@n16)))
   > n21 = Branch(Condition=Get bIsCrouching())
   n22 = Set leftMov_fvalue(execute=@n20.else, leftMov_fvalue:double=0) -> (Output_Get:double)
   n23 = MapRangeClamped(Value=FInterpTo(Current=@n22, Target=float + float(A=@n22, B=1), DeltaTime=@n13, InterpSpeed=4), InRangeA=0, InRangeB=1)
   n24 = Set leftMov_fvalue(execute=@n21.else, leftMov_fvalue=@n23)
   Set leftMov_fvalue(
    execute=@n21.then
    leftMov_fvalue=FInterpTo(Current=@n24, Target=float * float(A=@n23, B=0.92), DeltaTime=Get Delta Time X(), InterpSpeed=25)
   )
   n25 = Round(A=@n2.YMovement)
   n26 = NOT Boolean(A=Get bIsTargeting())
   n27 = AND Boolean(A=NOT Boolean(A=Get bIsJumping()), B=NOT Boolean(A=Get bIsCrouching()))
   n28 = Branch(execute=@n3.then_1, Condition=AND Boolean(A="To Boolean (Integer)"(InInt=@n25), B=AND Boolean(A=@n27, B=@n26)))
   > n29 = Branch(Condition=Get bIsTacticalSprinting())
   n30 = Set forwardMov_fvalue(execute=@n28.else, forwardMov_fvalue:double=0) -> (Output_Get:double)
   n31 = MapRangeClamped(Value=FInterpTo(Current=@n30, Target=float + float(A=@n30, B=1), DeltaTime=@n13, InterpSpeed=21), InRangeA=0, InRangeB=1, OutRangeB=0.5)
   n32 = Set forwardMov_fvalue(execute=@n29.else, forwardMov_fvalue=@n31)
   Set forwardMov_fvalue(
    execute=@n29.then
    forwardMov_fvalue=FInterpTo(Current=@n32, Target=float + float(A=@n31, B=0.5), DeltaTime=Get Delta Time X(), InterpSpeed=10)
   )
   Set newCrouching(execute=@n3.then_3, newCrouching:bool=OR Boolean(A=Get bIsTargeting(), B=Get bIsCrouching())) -> (Output_Get:bool)
   n33 = Branch(execute=@n3.then_4, Condition=Get bIsSkyDiving())
   > Set alphaSkyCond(alphaSkyCond:double=0.3) -> (Output_Get:double)
   > Set Override World Gravity(Override World Gravity:Vector="0,0,-250") -> (Output_Get:Vector)
   Set alphaSkyCond(execute=@n33.else, alphaSkyCond=0.8)
   > Set Override World Gravity(Override World Gravity="0,0,-980")
   n34 = GetOwnerSkeletalMeshComponent()
   n35 = "Clamp (Float)"(
    Value=BreakRotator(InRot=GetSocketTransform!(self=@n34, InSocketName=clavicle_l, TransformSpace=RTS_Component) -> (ReturnValue_Location:Vector, ReturnValue_Rotation:Rotator, ReturnValue_Scale).ReturnValue_Rotation).Yaw
    Min=-25
    Max=20
   )
   Set Return Value Rotation l(execute=@n3.then_5, Return Value Rotation l:Rotator=MakeRotator(Roll=@n35, Pitch=5, Yaw=@n35)) -> (Output_Get:Rotator)
   n36 = "Clamp (Float)"(
    Value=BreakRotator(InRot=GetSocketTransform(self=@n34, InSocketName=clavicle_r, TransformSpace=RTS_Component).ReturnValue_Rotation).Yaw
    Min=-20
    Max=25
   )
   > Set Return Value Rotation_r(Return Value Rotation_r:Rotator=MakeRotator(Roll=@n36, Pitch=-5, Yaw=@n36)) -> (Output_Get:Rotator)
   > n37 = Sequence!(execute) -> (then_0, then_1, then_2)
   > Set wingsCorrectIsParaCurVal(wingsCorrectIsParaCurVal:double=MapRangeClamped(
     Value=GetDeltaTransformFromRefPose!(self=@n34, BoneName=spine_05, BaseName=pelvis) -> (
      ReturnValue_Location:Vector
      ReturnValue_Rotation_Roll:float
      ReturnValue_Rotation_Pitch:float
      ReturnValue_Rotation_Yaw:float
      ReturnValue_Scale:Vector
     ).ReturnValue_Rotation_Yaw
     InRangeA=-37
     InRangeB=-49
    )) -> (Output_Get:double)
   > n38 = Branch(Condition="float > float"(A=Get wingsCorrectIsParaCurVal() -> (wingsCorrectIsParaCurVal:double)))
   > n39 = Set wingsCorrectIsPara_tempVal(wingsCorrectIsPara_tempVal:double=Get wingsCorrectIsParaCurVal()) -> (Output_Get:double)
   Set wingsCorrectIsPara_tempVal(execute=@n38.else, wingsCorrectIsPara_tempVal=@n39)
   > Delay(Duration=1)
   > Set wingsCorrectIsPara_tempVal(wingsCorrectIsPara_tempVal=0)
   n40 = Branch(execute=@n37.then_1, Condition=Get bIsInFrontEnd())
   > Set Component Linear Acc Scale(Component Linear Acc Scale:Vector="0.8,0.8,1") -> (Output_Get:Vector)
   > Set Para Glide Open Glider_0(Para Glide Open Glider_0:bool=false) -> (Output_Get:bool)
   n41 = Set Para Glide Open Glider(
    execute=@n40.else
    Para Glide Open Glider:bool=Get bParaGlide_OpenGlider!(self:Fort Player Anim Instance Object Reference=GetOwnerFortPlayerAnimInstance!(self:Custom Character Part Anim Instance Object Reference) -> (ReturnValue:Fort Player Anim Instance Object Reference)) -> (bParaGlide_OpenGlider:bool)
   ) -> (Output_Get:bool)
   > n42 = Branch(Condition=@n41)
   > Set Override World Gravity(Override World Gravity="0,0,-200")
   > Set Component Linear Acc Scale(Component Linear Acc Scale="0, 0, 0")
   > Delay(Duration=1.1)
   > Set Para Glide Open Glider_0(Para Glide Open Glider_0=true)
   Set Component Linear Acc Scale(execute=@n42.else, Component Linear Acc Scale="0.8,0.8,1")
   > Delay(Duration=0.6)
   > Set Para Glide Open Glider_0(Para Glide Open Glider_0=false)
   TryGetPawnOwner()
  ]}
]
variables [
 Alpha Output {type=number}
 alpha_rotSpine_inCrouching {type=number}
 alphaSkyCond {type=number}
 Component Linear Acc Scale {type=Vector, default=<0.8 0.8 1>}
 Delta Time X {type=number}
 fort_player {type=FortPlayerAnimInstance, default=null}
 forwardMov_fvalue {type=number}
 leftMov_fvalue {type=number}
 Melee_spineRotation X {type=number}
 newCrouching {type=boolean}
 Override World Gravity {type=Vector, default=<0 0 -800>}
 Para Glide Open Glider {type=boolean}
 Para Glide Open Glider_0 {type=boolean}
 Return Value Rotation l {type=Rotator}
 Return Value Rotation_r {type=Rotator}
 "Return ValueSpine_05 Rotation Z (Yaw)" {type=number}
 RightMov_fvalue {type=number}
 rotValue_FlightState_L {type=number}
 rotValue_FlightState_R {type=number}
 wingsCorrectIsPara_tempVal {type=number}
 wingsCorrectIsParaCurVal {type=number}
 X Movement_Pos {type=number}
 X MovementNeg {type=number}
]
